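# Translucent, lit material for WildEstate fountain water, usable on Niagara sprites and mesh particles.
# Noise-distorted, panning stripes alternate two time-of-day colors, screen-blended with foam highlights and topped with spigot foam as eye-adaptation-compensated emissive.
# The stripe coordinate also drives a sine normal-direction vertex offset, while particle dynamic parameters control a wobbly opacity wipe and uniform offset.
# Depth fade softens intersections, and MF_FN_DestructibleDamage wraps the final attributes with additive WPO.
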Material /WildEstate/Environment/Props/UsedCookie/UsedCookie_Fountain_A/Materials/M_WildEstate_FountainWater_01
graph {
 properties {
  bAutomaticallySetUsageInEditor=false
  bTangentSpaceNormal=false
  bUseMaterialAttributes=true
  bUsedWithNiagaraMeshParticles=true
  bUsedWithNiagaraSprites=true
  blendMode=BLEND_Translucent
  translucencyLightingMode=TLM_Surface
  usedShadingModels=Default Lit
 }
 nodes [
  n1 = TextureSampleParameter2D(UVs=Panner(Coordinate=TextureCoordinate() {vTiling=0.01}) {speedX=0.15, speedY=0.4}) {
   parameterName=Texture_Distortion
   samplerType=SAMPLERTYPE_LinearGrayscale
   texture=/Game/Packages/Fortress_Sky/SkyDome/Master/T_SkyDome_Cloud_Voroni
  }
  n2 = Add(
   A=Add(A=TextureCoordinate(), B=ScalarParameter() {defaultValue=-0.06, parameterName=Param_1}) {constB=-0.2}
   B=Multiply(A=@n1.R, B=ScalarParameter() {defaultValue=0.05})
  )
  n3 = Power(Base=OneMinus(None=ComponentMask(None=@n2) {g=true}), Exp=ScalarParameter() {defaultValue=40, parameterName=SpigotFoam_Power})
  n4 = ScalarParameter() {defaultValue=20, parameterName=SpigotFoam_Mult}
  n5 = Multiply(
   A=Multiply(A=VectorParameter() {parameterName=FoamColor, defaultValue=<rgba 0.36 0.54 0.9 1>}.RGB, B=NamedRerouteUsage())
   B=Clamp(None=Multiply(A=@n3, B=@n4) {constB=2})
  )
  n6 = MaterialFunctionCall(
   Color_Morning=VectorParameter() {group=Color_Base, parameterName=ColorA_Morning, defaultValue=<rgba 0.35 0.7 0.7 1>}.RGB
   Color_Day=VectorParameter() {group=Color_Base, parameterName=ColorA_Day, defaultValue=<rgba 0.35 0.7 0.7 1>}.RGB
   Color_Evening=VectorParameter() {group=Color_Base, parameterName=ColorA_Evening, defaultValue=<rgba 0.35 0.7 0.7 1>}.RGB
   Color_Night=VectorParameter() {group=Color_Base, parameterName=ColorA_night, defaultValue=<rgba 0.35 0.7 0.7 1>}.RGB
  ) {materialFunction=&Color_Time_of_Day}
  n7 = MaterialFunctionCall(
   Color_Morning=VectorParameter() {group=Color_Base, parameterName=ColorB_Morning, defaultValue=<rgba 0.1195 0.5711 0.5972 1>}.RGB
   Color_Day=VectorParameter() {group=Color_Base, parameterName=ColorB_Day, defaultValue=<rgba 0.1195 0.5711 0.5972 1>}.RGB
   Color_Evening=VectorParameter() {group=Color_Base, parameterName=ColorB_Evening, defaultValue=<rgba 0.1195 0.5711 0.5972 1>}.RGB
   Color_Night=VectorParameter() {group=Color_Base, parameterName=ColorB_night, defaultValue=<rgba 0.1195 0.5711 0.5972 1>}.RGB
  ) {materialFunction=&Color_Time_of_Day}
  n8 = TextureCoordinate()
  n9 = AppendVector(
   A=ComponentMask(None=@n8) {r=true}
   B=Multiply(A=ComponentMask(None=@n8) {g=true}, B=ScalarParameter() {defaultValue=4, parameterName=StripeTiling})
  )
  n10 = TextureSampleParameter2D(UVs=Panner(Coordinate=TextureCoordinate() {vTiling=0.01}) {speedX=0.075, speedY=0.09}) {
   parameterName=Texture_Distortion
   samplerType=SAMPLERTYPE_LinearGrayscale
   texture=&T_SkyDome_Cloud_Voroni
  }
  n11 = ScalarParameter() {defaultValue=0.8, parameterName=WiggleStrength}
  n12 = Multiply(A=@n10.R, B=@n11)
  n13 = Constant() {r=1}
  n14 = ScalarParameter() {defaultValue=-3.75, parameterName=StripePanspeed}
  n15 = ComponentMask(None=Panner(Coordinate=Add(A=@n9, B=@n12), Speed=AppendVector(A=@n13, B=@n14)) {speedY=-3}) {g=true}
  n16 = Modulo!(A=Round(None=Abs(None=@n15)), B=Constant() {r=2})
  n17 = TextureCoordinate() {uTiling=2}
  n18 = ScalarParameter() {defaultValue=0.25, parameterName=TopLayer_PanSpeed_X}
  n19 = ScalarParameter() {defaultValue=-1.55, parameterName=TopLayer_PanSpeed_Y}
  n20 = TextureSampleParameter2D(UVs=Panner(Coordinate=@n17, Speed=AppendVector(A=@n18, B=@n19)) {speedX=0.25, speedY=-2}) {parameterName=Texture_WaterHighlights, texture=/Game/Packages/Fortress_Transylvania/Haunted_Forest/Swamps/Textures/T_sea_foam_03}
  n21 = MaterialFunctionCall(
   Value_Morning=ScalarParameter() {defaultValue=0.5, parameterName=Morning}
   Value_Day=ScalarParameter() {defaultValue=1, parameterName=Day}
   Value_Evening=ScalarParameter() {defaultValue=0.5, parameterName=Evening}
   Value_Night=ScalarParameter() {defaultValue=0.3, parameterName=Night}
  ) {materialFunction=&Scalar_Time_of_Day}
  n22 = ComponentMask(None=@n17) {g=true}
  n23 = ScalarParameter() {defaultValue=4, parameterName=Param_2}
  n24 = LinearInterpolate(
   A=Multiply(
    A=ScalarParameter() {defaultValue=2, parameterName=TopLayer_Intensity_A}
    B=NamedRerouteDeclaration(None=@n21.Result) {name=TimeOfDay, nodeColor=<rgba 1 0.8438 0 1>}
   )
   B=ScalarParameter() {defaultValue=0.15, parameterName=TopLayer_Intensity_B}
   Alpha=Clamp(None=Multiply(A=@n22, B=@n23))
  )
  n25 = Clamp(None=MaterialFunctionCall!(Blend=Multiply(A=@n20.G, B=@n24), Base=LinearInterpolate(A=@n6.Result, B=@n7.Result, Alpha=@n16)) {materialFunction=/Engine/Functions/Engine_MaterialFunctions03/Blends/Blend_Screen}.Result)
  n26 = NamedRerouteUsage()
  n27 = ScalarParameter() {defaultValue=0.75, parameterName=Opacity}
  n28 = Constant() {r=1.5}
  n29 = AppendVector(A=AppendVector(A=Power(Base=@n10.R, Exp=@n28), B=Constant() {r=0.5}), B=Constant() {r=1})
  n30 = TextureCoordinate() {uTiling=2, vTiling=0.6}
  n31 = MaterialFunctionCall(
   Normal=TextureSample(UVs=Panner(Coordinate=@n30) {speedX=0.15, speedY=-0.5}) {samplerType=SAMPLERTYPE_Normal, texture=/Game/Tools/WaterfallGenerator/Art/T_Waterfall_NormalB}.RGB
   Flatness=ScalarParameter() {defaultValue=0.4, parameterName=DetailNormal_Flatness}
  ) {materialFunction=&FlattenNormal}
  n32 = Sine(None=@n15)
  n33 = Multiply(A=Multiply(A=Add(A=@n32)) {constB=0.5}, B=ScalarParameter() {defaultValue=3, parameterName=WPO_Intensity}) {constB=100}
  n34 = DynamicParameter() {defaultValue=<rgba 0 0 1 1>, paramNames [Alpha_Wipe, WPO_UniformOffset, Param3, Param4]}
  n35 = SetMaterialAttributes!(
   MaterialAttributes
   Base Color=ScalarParameter() {defaultValue=0.025, parameterName=ReflectionBrightness}
   Specular=ScalarParameter() {defaultValue=1, parameterName=Spec}
   Roughness=ScalarParameter() {defaultValue=0.2, parameterName=Roughness}
   Emissive Color=EyeAdaptationInverse(LightValueInput=Add(A=@n5, B=@n25)).EyeAdaptationInverse
   Opacity=DepthFade(Opacity=Multiply(A=@n26, B=@n27), FadeDistance=ScalarParameter() {defaultValue=50, parameterName=DepthFade_Dist})
   Normal=Transform(None=MaterialFunctionCall(BaseNormal=@n29, AdditionalNormal=@n31.Result) {materialFunction=&BlendAngleCorrectedNormals}.Result)
   Metallic=ScalarParameter() {defaultValue=1, parameterName=Metallic}
   World Position Offset=Multiply(A=Subtract(A=@n33, B=@n34.WPO_UniformOffset), B=VertexNormalWS())
  ) {attributeSetTypes [
    #1
    #2
    #3
    #4
    #5
    #6
    #7
    #8
   ]}
  MaterialFunctionCall(Material_In=@n35, UseCrackedDamageTextures=StaticBool(), UseAdditiveWPO=StaticBool() {value=true}) {materialFunction=&MF_FN_DestructibleDamage}
  n36 = Add(A=TextureCoordinate()) {constB=0.45}
  n37 = DynamicParameter() {defaultValue=<rgba 0 0 1 1>, paramNames [Alpha_Wipe, WPO_UniformOffset, Param3, Param4]}
  n38 = Multiply(A=@n11, B=Constant() {r=0.5}) {constB=0.5}
  n39 = ComponentMask(None=Add(A=Subtract(A=@n36, B=@n37.Alpha_Wipe), B=Multiply(A=@n12, B=@n38))) {g=true}
  NamedRerouteDeclaration(None=Clamp(None=Round(None=@n39))) {name=Alpha_Wipe, nodeColor=<rgba 0 0.0859 1 1>}
 ]
}
